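# Sound cue for a glacier creak ambient spot one-shot in the Helios snow biome.
# An equally weighted Random node picks one of seven glacier creak waves reused from the Asteria RealityWarrior biome, then a Modulator varies volume and slightly raises pitch.
# It uses the tile ambient one-shot linear attenuation and outdoor ambient sound class, sending to an exterior ambient source bus and a distant reverb submix.
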SoundCue /Game/Sounds/Helios/Biomes/Snow/Spots/Spot_Snow_Glacier_Creak_Cue
root=Modulator_0
nodes [
 {id=Modulator_0, children [Random_1], properties {pitchMax=1.1, pitchMin=1, volumeMin=0.6}}
 {
  id=Random_1
  children [WavePlayer_0, WavePlayer_6, WavePlayer_7, WavePlayer_8, WavePlayer_9, WavePlayer_12, WavePlayer_13]
  properties {weights [1,1,1,1,1,1,1]}
 }
 {id=WavePlayer_0, properties {soundWaveAssetPtr=/Game/Sounds/Asteria/Biomes/RealityWarrior/Spots/Spot_Glacier_Creak_01}}
 {id=WavePlayer_12, properties {soundWaveAssetPtr=/Game/Sounds/Asteria/Biomes/RealityWarrior/Spots/Spot_Glacier_Creak_06}}
 {id=WavePlayer_13, properties {soundWaveAssetPtr=/Game/Sounds/Asteria/Biomes/RealityWarrior/Spots/Spot_Glacier_Creak_07}}
 {id=WavePlayer_6, properties {soundWaveAssetPtr=/Game/Sounds/Asteria/Biomes/RealityWarrior/Spots/Spot_Glacier_Creak_02}}
 {id=WavePlayer_7, properties {soundWaveAssetPtr=/Game/Sounds/Asteria/Biomes/RealityWarrior/Spots/Spot_Glacier_Creak_03}}
 {id=WavePlayer_8, properties {soundWaveAssetPtr=/Game/Sounds/Asteria/Biomes/RealityWarrior/Spots/Spot_Glacier_Creak_04}}
 {id=WavePlayer_9, properties {soundWaveAssetPtr=/Game/Sounds/Asteria/Biomes/RealityWarrior/Spots/Spot_Glacier_Creak_05}}
]
properties {
 attenuationSettings=/Game/Sounds/Attenuation/Fort_TileAmbient_OneShot_Linear_Att
 maxDistance=8000
 soundClassObject=&Fort_Ambient_OneShots_Outdoors
 volumeMultiplier=0.376
 busSends [{audioBus=null, minSendLevel=0.6, sendLevel=1, soundSourceBus=&AmbientAudio_Exterior_SourceBus}]
 soundSubmixSends [{minSendLevel=1, soundSubmix=&Fortnite_Reverb_Distant}]
}
