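# Looping outdoor ambient cue for glacier creak spot sounds in the snow biome, reusing Asteria RealityWarrior waves and attenuations.
# After each random delay, a mixer layers a random ice-lake sound with a quieter, pitched-down random ice-pressure detail.
# Each layer has its own pitch/volume modulation and attenuation.
# The quality switch feeds that mix to its first two slots and leaves the other two empty.
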SoundCue /Game/Sounds/Helios/Biomes/Snow/Spots/Spot_Snow_Glacier_Creak_Looping_Cue
root=Looping_0
nodes [
 {
  id=Attenuation_0
  children [Random_0]
  properties {attenuationSettings=/Game/Sounds/Asteria/Biomes/RealityWarrior/Spots/Spot_Glacier_Creak_Capsule_Linear_Att}
 }
 {
  id=Attenuation_1
  children [Random_1]
  properties {attenuationSettings=/Game/Sounds/Asteria/Biomes/RealityWarrior/Spots/Spot_Glacier_Creak_Capsule_Linear_Details_Att}
 }
 {id=Delay_0, children [QualityLevel_0], properties {delayMax=25, delayMin=12}}
 {id=Looping_0, children [Delay_0]}
 {id=Mixer_0, children [Modulator_1,Modulator_0], properties {inputVolume [1,0.5]}}
 {id=Modulator_0, children [Attenuation_1], properties {pitchMax=0.5, pitchMin=0.4, volumeMax=0.5, volumeMin=0.4}}
 {id=Modulator_1, children [Attenuation_0], properties {pitchMax=1.1, pitchMin=0.9, volumeMin=0.85}}
 {id=QualityLevel_0, children [Mixer_0,Mixer_0,null,null]}
 {
  id=Random_0
  children [WavePlayer_1, WavePlayer_2, WavePlayer_3, WavePlayer_4, WavePlayer_5, WavePlayer_10, WavePlayer_11]
  properties {weights [1,1,1,1,1,1,1]}
 }
 {
  id=Random_1
  children [WavePlayer_0, WavePlayer_6, WavePlayer_7, WavePlayer_8, WavePlayer_9, WavePlayer_12, WavePlayer_13, WavePlayer_14]
  properties {weights [1,1,1,1,1,1,1,1]}
 }
 {id=WavePlayer_0, properties {soundWaveAssetPtr=/Game/Sounds/Asteria/Biomes/RealityWarrior/Spots/Spot_Ice_Pressure_Close_02}}
 {id=WavePlayer_1, properties {soundWaveAssetPtr=/Game/Sounds/Asteria/Biomes/RealityWarrior/Spots/Spot_Snow_Ice_Lake_07}}
 {id=WavePlayer_10, properties {soundWaveAssetPtr=/Game/Sounds/Asteria/Biomes/RealityWarrior/Spots/Spot_Snow_Ice_Lake_05}}
 {id=WavePlayer_11, properties {soundWaveAssetPtr=/Game/Sounds/Asteria/Biomes/RealityWarrior/Spots/Spot_Snow_Ice_Lake_06}}
 {id=WavePlayer_12, properties {soundWaveAssetPtr=/Game/Sounds/Asteria/Biomes/RealityWarrior/Spots/Spot_Ice_Pressure_Close_04}}
 {id=WavePlayer_13, properties {soundWaveAssetPtr=/Game/Sounds/Asteria/Biomes/RealityWarrior/Spots/Spot_Ice_Pressure_Close_05}}
 {id=WavePlayer_14, properties {soundWaveAssetPtr=/Game/Sounds/Asteria/Biomes/RealityWarrior/Spots/Spot_Ice_Pressure_Close_07}}
 {id=WavePlayer_2, properties {soundWaveAssetPtr=/Game/Sounds/Asteria/Biomes/RealityWarrior/Spots/Spot_Snow_Ice_Lake_01}}
 {id=WavePlayer_3, properties {soundWaveAssetPtr=/Game/Sounds/Asteria/Biomes/RealityWarrior/Spots/Spot_Snow_Ice_Lake_02}}
 {id=WavePlayer_4, properties {soundWaveAssetPtr=/Game/Sounds/Asteria/Biomes/RealityWarrior/Spots/Spot_Snow_Ice_Lake_03}}
 {id=WavePlayer_5, properties {soundWaveAssetPtr=/Game/Sounds/Asteria/Biomes/RealityWarrior/Spots/Spot_Snow_Ice_Lake_04}}
 {id=WavePlayer_6, properties {soundWaveAssetPtr=/Game/Sounds/Asteria/Biomes/RealityWarrior/Spots/Spot_Ice_Pressure_Close_06}}
 {id=WavePlayer_7, properties {soundWaveAssetPtr=/Game/Sounds/Asteria/Biomes/RealityWarrior/Spots/Spot_Ice_Pressure_Close_08}}
 {id=WavePlayer_8, properties {soundWaveAssetPtr=/Game/Sounds/Asteria/Biomes/RealityWarrior/Spots/Spot_Ice_Pressure_Close_01}}
 {id=WavePlayer_9, properties {soundWaveAssetPtr=/Game/Sounds/Asteria/Biomes/RealityWarrior/Spots/Spot_Ice_Pressure_Close_03}}
]
properties {
 attenuationSettings=&Spot_Glacier_Creak_Capsule_Linear_Att
 maxDistance=5000
 soundClassObject=&Fort_Ambient_Outdoors_Athena
 volumeMultiplier=0.301
 preEffectBusSends [{}]
}
