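# Snow biome one-shot ambient cue for medium-strength wind spots, built from reused Apollo desert gust waves.
# A quality switch sends its first two levels to a random pool of fifteen sandy and light-debris gusts, and its last two to four of those gusts.
# Each pool passes through a pitch and volume modulator, with wider volume variation on the smaller one.
# Outdoor one-shot class and wide tile-ambient attenuation apply.
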SoundCue /Game/Sounds/Helios/Biomes/Snow/Spots/Spot_Wind_Medium_Strength_Cue
root=QualityLevel_0
nodes [
 {id=Modulator_0, children [Random_2], properties {pitchMax=1.1, pitchMin=0.7, volumeMax=1.1, volumeMin=0.5}}
 {id=Modulator_1, children [Random_0], properties {pitchMax=1.1, pitchMin=0.7, volumeMax=1.1, volumeMin=0.9}}
 {id=QualityLevel_0, children [Modulator_1,Modulator_1,Modulator_0,Modulator_0]}
 {
  id=Random_0
  children [
   WavePlayer_1
   WavePlayer_11
   WavePlayer_12
   WavePlayer_13
   WavePlayer_14
   WavePlayer_15
   WavePlayer_16
   WavePlayer_20
   WavePlayer_21
   WavePlayer_22
   WavePlayer_23
   WavePlayer_24
   WavePlayer_39
   WavePlayer_43
   WavePlayer_44
  ]
  properties {weights [1,1,1,1,1,1,1,1,1,1,1,1,1,1,1]}
 }
 {
  id=Random_2
  children [WavePlayer_46, WavePlayer_45, WavePlayer_47, WavePlayer_48]
  properties {bShouldExcludeFromBranchCulling=true, weights [1,1,1,1]}
 }
 {id=WavePlayer_1, properties {soundWaveAssetPtr=/Game/Sounds/Apollo/Biomes/Desert/AMB_SS_S15_Desert_WindGust_Med_hv_Sandy_10}}
 {id=WavePlayer_11, properties {soundWaveAssetPtr=/Game/Sounds/Apollo/Biomes/Desert/AMB_SS_S15_Desert_WindGust_Med_hv_Sandy_01}}
 {id=WavePlayer_12, properties {soundWaveAssetPtr=/Game/Sounds/Apollo/Biomes/Desert/AMB_SS_S15_Desert_WindGust_Med_hv_Sandy_02}}
 {id=WavePlayer_13, properties {soundWaveAssetPtr=/Game/Sounds/Apollo/Biomes/Desert/AMB_SS_S15_Desert_WindGust_Med_hv_Sandy_03}}
 {id=WavePlayer_14, properties {soundWaveAssetPtr=/Game/Sounds/Apollo/Biomes/Desert/AMB_SS_S15_Desert_WindGust_Med_hv_Sandy_04}}
 {id=WavePlayer_15, properties {soundWaveAssetPtr=/Game/Sounds/Apollo/Biomes/Desert/AMB_SS_S15_Desert_WindGust_Med_hv_Sandy_05}}
 {id=WavePlayer_16, properties {soundWaveAssetPtr=/Game/Sounds/Apollo/Biomes/Desert/AMB_SS_S15_Desert_WindGust_Med_hv_Sandy_06}}
 {id=WavePlayer_20, properties {soundWaveAssetPtr=/Game/Sounds/Apollo/Biomes/Desert/AMB_SS_S15_Desert_WindGust_Med_hv_Sandy_07}}
 {id=WavePlayer_21, properties {soundWaveAssetPtr=/Game/Sounds/Apollo/Biomes/Desert/AMB_SS_S15_Desert_WindGust_Med_hv_Sandy_08}}
 {id=WavePlayer_22, properties {soundWaveAssetPtr=/Game/Sounds/Apollo/Biomes/Desert/AMB_SS_S15_Desert_WindGust_Med_hv_Sandy_09}}
 {id=WavePlayer_23, properties {soundWaveAssetPtr=/Game/Sounds/Apollo/Biomes/Desert/AMB_SS_S15_Desert_WindGust_Med_Lt_Debris_01}}
 {id=WavePlayer_24, properties {soundWaveAssetPtr=/Game/Sounds/Apollo/Biomes/Desert/AMB_SS_S15_Desert_WindGust_Med_Lt_Debris_02}}
 {id=WavePlayer_39, properties {soundWaveAssetPtr=/Game/Sounds/Apollo/Biomes/Desert/AMB_SS_S15_Desert_WindGust_Med_Lt_Debris_03}}
 {id=WavePlayer_43, properties {soundWaveAssetPtr=/Game/Sounds/Apollo/Biomes/Desert/AMB_SS_S15_Desert_WindGust_Med_Lt_Debris_04}}
 {id=WavePlayer_44, properties {soundWaveAssetPtr=/Game/Sounds/Apollo/Biomes/Desert/AMB_SS_S15_Desert_WindGust_Med_Lt_Debris_05}}
 {id=WavePlayer_45, properties {soundWaveAssetPtr=&AMB_SS_S15_Desert_WindGust_Med_hv_Sandy_02}}
 {id=WavePlayer_46, properties {soundWaveAssetPtr=&AMB_SS_S15_Desert_WindGust_Med_hv_Sandy_01}}
 {id=WavePlayer_47, properties {soundWaveAssetPtr=&AMB_SS_S15_Desert_WindGust_Med_Lt_Debris_01}}
 {id=WavePlayer_48, properties {soundWaveAssetPtr=&AMB_SS_S15_Desert_WindGust_Med_Lt_Debris_02}}
]
properties {
 attenuationSettings=/Game/Sounds/Attenuation/Fort_TileAmbient_OneShot_Wider_Att
 maxDistance=8000
 soundClassObject=&Fort_Ambient_OneShots_Outdoors
 volumeMultiplier=0.301
 preEffectBusSends [{}]
}
